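; Vault monitor prop Blueprint whose screen look is chosen per instance by a MonitorType variable.
; The construction script uses a Select on MonitorType to pick one of the FrontEnd monitor screen materials and applies it to material slot 1 of the static mesh component.
; The first and last of the eight options share the same screen material, giving seven distinct screens.
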(fn ConstructionScript
  (Rendering|Material|SetMaterial (Variables|Debug|GetStaticMeshComponent) 1 (Utilities|Select :Index (Variables|Default|GetMonitorType) :Option%200 "/Game/Environments/FrontEnd/Materials/M_Monitor_Screen_V_l0.M_Monitor_Screen_V_l0" :Option%201 "/Game/Environments/FrontEnd/Materials/M_Monitor_Screen_V_I1.M_Monitor_Screen_V_I1" :Option%202 "/Game/Environments/FrontEnd/Materials/M_Monitor_Screen_V_I2.M_Monitor_Screen_V_I2" :Option%203 "/Game/Environments/FrontEnd/Materials/M_Monitor_Screen_V_I3.M_Monitor_Screen_V_I3" :Option%204 "/Game/Environments/FrontEnd/Materials/M_Monitor_Screen_V_I4.M_Monitor_Screen_V_I4" :Option%205 "/Game/Environments/FrontEnd/Materials/M_Monitor_Screen_V_I5.M_Monitor_Screen_V_I5" :Option%206 "/Game/Environments/FrontEnd/Materials/M_Monitor_Screen_V_I6.M_Monitor_Screen_V_I6" :Option%207 "/Game/Environments/FrontEnd/Materials/M_Monitor_Screen_V_l0.M_Monitor_Screen_V_l0")))
